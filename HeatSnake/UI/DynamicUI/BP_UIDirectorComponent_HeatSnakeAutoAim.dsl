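; UI director component for HeatSnake that shows or hides the smart reticle UI scene in response to auto-aim events.
; When the UI is ready, it starts listening on the global GameplayEventRouter for the HeatSnake auto-aim event struct.
; A delta gate compares the payload against the Should Show variable, adding or removing the scene only on changes.
; End play stops the listener and removes the scene.
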(event EventOnReadyToShowUI
  (bind (_asynctaskproxy _payload) (GameplayEventRouter|StartListeningToEvent (GameplayEventRouter|GetGlobalEventRouter) "/HeatSnake/UI/DynamicUI/ST_HeatSnake_AutoAimEventData.ST_HeatSnake_AutoAimEventData")
    (:then
      (Variables|Default|SetOnEnabledSmartReticle _asynctaskproxy))
    (:OnEventReceived
      (bind _updated_variable (Utilities|FlowControl|DeltaGateWildcard (Variables|Default|GetShould%20Show) (Utilities|Struct|BreakSTHeatSnakeAutoAimEventData _payload))
        (:Success
          (CallFunction|AddorRemoveScene (Variables|Default|GetSmartReticleEnabled) _updated_variable))))))

(event EventEndPlay
  (params (EndPlayReason :type EEndPlayReason%20Enum :default "Destroyed"))
  (GameplayEventRouter|StopListeningtoEvent (Variables|Default|GetOnEnabledSmartReticle))
  (CallFunction|RemoveScene (Variables|Default|GetSmartReticleEnabled)))
